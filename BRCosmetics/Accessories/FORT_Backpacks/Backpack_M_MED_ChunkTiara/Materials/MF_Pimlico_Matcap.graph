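# Material function for the ChunkTiara backpack cosmetic that outputs a matcap color, by default from a Pimlico rim mask texture.
# It crosses the normalized view-space pixel position with the view-space vertex normal for a perspective-aware lookup.
# Only the green component of that result feeds both U and V, scaled and offset by half, so sampling follows the texture's diagonal.
# The MatcapTexture parameter's RGB is the output.
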MaterialFunction /BRCosmetics/Accessories/FORT_Backpacks/Backpack_M_MED_ChunkTiara/Materials/MF_Pimlico_Matcap
graph {nodes [
  n1 = TransformPosition(None=WorldPosition().XYZ) {transformSourceType=TRANSFORMPOSSOURCE_World, transformType=TRANSFORMPOSSOURCE_View}
  n2 = VertexNormalWS()
  n3 = MaterialFunctionCall(Color=CrossProduct(A=Normalize(VectorInput=@n1), B=Transform(None=@n2) {transformSourceType=TRANSFORMSOURCE_World, transformType=TRANSFORM_View})) {materialFunction=&SplitComponents}
  n4 = AppendVector(A=@n3.G, B=@n3.G)
  n5 = Constant2Vector() {g=0.5, r=0.5}
  n6 = TextureSampleParameter2D(UVs=Add(A=Multiply(A=@n4, B=@n5)) {constB=0.5}) {
   group=Rim
   parameterName=MatcapTexture
   samplerType=SAMPLERTYPE_Masks
   texture=/BRCosmetics/Accessories/FORT_Backpacks/Backpack_M_MED_ChunkTiara/Textures/T_Pimlico_BaseBean_Rim
  }
  FunctionOutput(None=NamedRerouteDeclaration(None=@n6.RGB) {name=MatcapTexture, nodeColor=<rgba 0.6406 1 0 1>})
 ]}
