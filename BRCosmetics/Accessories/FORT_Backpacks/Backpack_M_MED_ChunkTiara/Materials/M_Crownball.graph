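# Subsurface material for the Crownball of the ChunkTiara backpack cosmetic, a stylized character with an animated face.
# Base color layers albedo (baked or team-tinted), SubUV flipbook mouth and offsettable pupils, plus a thresholded blink mask, all parameter-driven.
# Sky-light-to-normal dot products drive screen-space halftone, scaled by camera distance, and etched shadows.
# Emissive sums masked self-illumination, matcap rim color and a switchable VFX fresnel glow.
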Material /BRCosmetics/Accessories/FORT_Backpacks/Backpack_M_MED_ChunkTiara/Materials/M_Crownball
graph {
 properties {
  bAutomaticallySetUsageInEditor=false
  bUsedWithMeshDeformer=true
  bUsedWithMorphTargets=true
  bUsedWithSkeletalMesh=true
  shadingModel=MSM_Subsurface
  usedShadingModels=Subsurface
 }
 nodes [
  n1 = Fresnel(ExponentIn=ScalarParameter() {defaultValue=1, parameterName=VFX Fresnel Exponent})
  n2 = ScalarParameter() {defaultValue=1.8, parameterName=VFX Fresnel Mult}
  n3 = Saturate(None=Power(Base=Multiply(A=@n1, B=@n2), Exp=ScalarParameter() {defaultValue=15, parameterName=VFX Fresnel Power}))
  n4 = VectorParameter() {parameterName=VFX Fresnel Color, defaultValue=<rgba 100 64.1515 0 1>}
  n5 = NamedRerouteUsage()
  n6 = ScalarParameter() {group=Shading, parameterName=SelfIllumPercentage}
  n7 = Multiply(A=NamedRerouteUsage(), B=Add(A=NamedRerouteUsage(), B=NamedRerouteUsage()))
  n8 = NamedRerouteUsage()
  n9 = Multiply(
   A=OneMinus(None=NamedRerouteUsage())
   B=Multiply(A=Multiply(A=@n7, B=@n8), B=ScalarParameter() {group=Shading, parameterName=EmissiveIntensity})
  )
  n10 = MaterialFunctionCall() {materialFunction=&/MF_Pimlico_Matcap}
  n11 = LinearInterpolate(
   A=Constant()
   B=Multiply(
    A=MaterialFunctionCall(Color=@n10.Result) {materialFunction=&SplitComponents}.G
    B=VectorParameter() {group=Rim, parameterName=RimColor, defaultValue=<rgba 0 0.7184 1 1>}.RGB
   )
   Alpha=ScalarParameter() {defaultValue=1, group=Rim, parameterName=HasRim}
  )
  Add(
   A=Multiply(A=ScalarParameter() {parameterName=VFXFresnelOnOff}, B=Multiply(A=@n3, B=@n4.RGB))
   B=Add(A=LinearInterpolate(B=@n5, Alpha=@n6) {constB=0}, B=Add(A=@n9, B=@n11))
  )
  LinearInterpolate(A=Add(A=NamedRerouteUsage(), B=NamedRerouteUsage()), Alpha=NamedRerouteUsage()) {constB=0.5}
  n12 = VectorParameter() {group=Shading, parameterName=TeamColor, defaultValue=<rgba 0 0.2666 0.4948 1>}
  LinearInterpolate(
   A=@n12.RGB
   B=VectorParameter() {group=Shading, parameterName=SecondaryTeamColor, defaultValue=<rgba 0 0.2666 0.4948 1>}.RGB
   Alpha=NamedRerouteUsage()
  )
  Multiply(A=NamedRerouteUsage(), B=VectorParameter() {group=Shading, parameterName=SSSColorMultiplier, defaultValue=<rgba 0.3733 0.1614 0.2968 1>}.RGB)
  n13 = TextureSampleParameter2D() {
   group=Textures
   parameterName=MasksTexture
   samplerType=SAMPLERTYPE_Masks
   texture=/BRCosmetics/Accessories/FORT_Backpacks/Backpack_M_MED_ChunkTiara/Textures/T_Pimlico_Crownball_ERTM
  }
  NamedRerouteDeclaration(None=@n13.R) {name=EmissiveMask, nodeColor=<rgba 1 0 0.7266 1>}
  n14 = NamedRerouteUsage()
  n15 = Multiply(
   A=Step(
    Y=TextureSample(Tex=@n14) {constCoordinate=2, texture=/BRCosmetics/Accessories/FORT_Backpacks/Backpack_M_MED_ChunkTiara/Textures/T_Pimlico_Crownball_AM}.A
    X=ScalarParameter() {group=Animation, parameterName=Blink}
   )
   B=NamedRerouteUsage()
  )
  NamedRerouteDeclaration(None=@n15) {name=BlinkMask, nodeColor=<rgba 0 1 0.1562 1>}
  n16 = Constant() {r=512}
  n17 = Divide(A=ScreenPosition().PixelPosition, B=AppendVector(A=@n16, B=@n16))
  n18 = ScalarParameter() {defaultValue=5, group=HalfToneEffect, parameterName=HalftoneEffectTiling}
  n19 = Distance(A=ActorPositionWS(), B=CameraPositionWS())
  n20 = TextureObjectParameter() {
   group=Texture
   parameterName=HalftonePattern
   samplerType=SAMPLERTYPE_Masks
   texture=/BRCosmetics/Accessories/FORT_Backpacks/Backpack_M_MED_ChunkTiara/Textures/T_Pimlico_BaseBean_Halftone
  }
  n21 = TextureSample(
   UVs=MaterialFunctionCall(UVs=Multiply(A=@n17, B=@n18) {constB=5}, Texture Scale=Divide(B=@n19) {constA=300}) {materialFunction=&ScaleUVsByCenter}.UVs
   Tex=@n20
  ) {samplerType=SAMPLERTYPE_LinearGrayscale}
  n22 = Clamp(
   None=OneMinus(None=NamedRerouteUsage())
   Min=ScalarParameter() {defaultValue=0.4, group=HalfToneEffect, parameterName=HalftoneSizeGradientMin}
   Max=ScalarParameter() {defaultValue=0.7, group=HalfToneEffect, parameterName=HalftoneSizeGradientMax}
  ) {maxDefault=0.75, minDefault=0.5}
  NamedRerouteDeclaration(None=OneMinus(None=SmoothStep(Max=@n21.G, Value=@n22))) {name=ScreenSpaceHalftone, nodeColor=<rgba 0.4766 1 0 1>}
  NamedRerouteDeclaration(None=@n21.B) {name=Etching, nodeColor=<rgba 1 0 0.9844 1>}
  n23 = TextureSampleParameter2D() {
   group=Textures
   parameterName=Masks
   samplerType=SAMPLERTYPE_Masks
   texture=/BRCosmetics/Accessories/FORT_Backpacks/Backpack_M_MED_ChunkTiara/Textures/T_Pimlico_Crownball_EyeMask
  }
  NamedRerouteDeclaration(None=@n23.G) {name=EyeMask, nodeColor=<rgba 1 0.1641 0 1>}
  NamedRerouteDeclaration(None=@n23.R) {name=RimMask, nodeColor=<rgba 0 0.1562 1 1>}
  NamedRerouteDeclaration(None=@n13.G) {name=RoughnessMask, nodeColor=<rgba 0 1 0.9531 1>}
  NamedRerouteDeclaration(None=@n13.B) {name=TeamColorMask, nodeColor=<rgba 1 0.6328 0 1>}
  NamedRerouteDeclaration(None=@n12.RGB) {name=TeamColor, nodeColor=<rgba 0.3125 0 1 1>}
  n24 = TextureObjectParameter() {
   group=Textures
   parameterName=MouthAnimAtlas
   texture=/BRCosmetics/Accessories/FORT_Backpacks/Backpack_M_MED_ChunkTiara/Textures/T_Pimlico_CrownBall_MouthAnim
  }
  n25 = MaterialFunctionCall!(
   Texture=@n24
   UVs=TextureCoordinate() {coordinateIndex=1}
   SubImages=AppendVector(
    A=ScalarParameter() {defaultValue=4, group=Textures, parameterName=MouthAnim_Columns}
    B=ScalarParameter() {defaultValue=4, group=Textures, parameterName=MouthAnim_Rows}
   )
   Frame=Floor(None=ScalarParameter() {group=Animation, parameterName=Frame_MouthAnim, sliderMin=0.1})
  ) {materialFunction=/Engine/Functions/Engine_MaterialFunctions03/Texturing/SubUV_Function}
  NamedRerouteDeclaration(None=@n25.Alpha) {name=MouthAnimAlpha, nodeColor=<rgba 0.0078 1 0 1>}
  NamedRerouteDeclaration(None=@n13.A) {name=EyeDirectionMask, nodeColor=<rgba 0 1 0.0625 1>}
  n26 = TextureObjectParameter() {
   group=Textures
   parameterName=PupilTexture
   texture=/BRCosmetics/Accessories/FORT_Backpacks/Backpack_M_MED_ChunkTiara/Textures/T_Pimlico_CrownBall_EyesAnim
  }
  n27 = MaterialFunctionCall(Float2=TextureCoordinate() {coordinateIndex=2}) {materialFunction=&BreakOutFloat2Components}
  n28 = ScalarParameter() {defaultValue=0.12, group=Animation, parameterName=RightEye_HorizontalMovement}
  n29 = ScalarParameter() {group=Animation, parameterName=RightEye_VerticalMovement}
  n30 = AppendVector(
   A=ScalarParameter() {defaultValue=2, group=Textures, parameterName=Eye_Columns}
   B=ScalarParameter() {defaultValue=2, group=Textures, parameterName=Eye_Rows}
  )
  n31 = Floor(None=ScalarParameter() {group=Animation, parameterName=Frame_EyeAnim, sliderMin=0.1})
  n32 = MaterialFunctionCall(Texture=@n26, UVs=AppendVector(A=Add(A=@n27.R, B=@n28), B=Add(A=@n27.G, B=@n29)), SubImages=@n30, Frame=@n31) {materialFunction=&SubUV_Function}
  n33 = MaterialFunctionCall(Float2=TextureCoordinate() {coordinateIndex=2}) {materialFunction=&BreakOutFloat2Components}
  n34 = ScalarParameter() {defaultValue=0.12, group=Animation, parameterName=LeftEye_HorizontalMovement}
  n35 = ScalarParameter() {group=Animation, parameterName=LeftEye_VerticalMovement}
  n36 = MaterialFunctionCall(Texture=@n26, UVs=AppendVector(A=Add(A=@n33.R, B=@n34), B=Add(A=@n33.G, B=@n35)), SubImages=@n30, Frame=@n31) {materialFunction=&SubUV_Function}
  n37 = NamedRerouteUsage()
  n38 = LinearInterpolate(A=@n32.Alpha, B=@n36.Alpha, Alpha=@n37)
  NamedRerouteDeclaration(None=@n38) {name=PupilAlpha, nodeColor=<rgba 0.3125 1 0 1>}
  n39 = VectorParameter() {group=HalfToneEffect, parameterName=EtchingColor, defaultValue=<rgba 0.2191 0.0039 0.3212 1>}
  n40 = NamedRerouteUsage()
  n41 = DotProduct(A=SkyAtmosphereLightDirection(), B=VertexNormalWS())
  n42 = Power(
   Base=Multiply(A=OneMinus(None=@n41)) {constB=0.3}
   Exp=ScalarParameter() {defaultValue=10, group=HalfToneEffect, parameterName=ShadowIntensity}
  ) {constExponent=10}
  n43 = Power(Base=MaterialFunctionCall(In=Multiply(A=@n40, B=@n42), Contrast=Constant() {r=0.004}) {materialFunction=&CheapContrast}.Result) {constExponent=0.1}
  n44 = ScalarParameter() {defaultValue=10, group=HalfToneEffect, parameterName=HaltoneAmount}
  n45 = NamedRerouteDeclaration(None=Saturate(None=Power(Base=@n41, Exp=@n44) {constExponent=20})) {name=HalftoneSizeGradient, nodeColor=<rgba 0.5469 0 1 1>}
  n46 = NamedRerouteUsage()
  n47 = NamedRerouteUsage()
  n48 = TextureObjectParameter() {
   group=Textures
   parameterName=BaseColorTexture
   texture=&T_Pimlico_Crownball_AM
  }
  n49 = TextureSample(Tex=@n48) {texture=&T_Pimlico_Crownball_AM}
  n50 = LinearInterpolate(B=NamedRerouteUsage(), Alpha=NamedRerouteUsage()) {constA=1}
  n51 = LinearInterpolate(
   A=LinearInterpolate(
    A=Multiply(A=@n49.RGB, B=@n50)
    B=TextureSample() {texture=/BRCosmetics/Accessories/FORT_Backpacks/Backpack_M_MED_ChunkTiara/Textures/T_Pimlico_BakedCrownball_AM}.RGB
    Alpha=ScalarParameter() {defaultValue=1, group=Shading, parameterName=AlbedoIsBaked}
   )
   B=@n25.RGB
   Alpha=@n25.RGB
  )
  n52 = LinearInterpolate(A=@n32.RGB, B=@n36.RGB, Alpha=@n37)
  n53 = NamedRerouteDeclaration(None=LinearInterpolate(A=LinearInterpolate(A=@n51, B=@n52, Alpha=@n38), B=Multiply(A=@n47) {constB=0.75}, Alpha=@n15)) {name=BaseColor, nodeColor=<rgba 0.1719 1 0 1>}
  n54 = Multiply(A=Multiply(A=Multiply(A=@n45, B=@n46), B=ScalarParameter() {defaultValue=1, group=HalfToneEffect, parameterName=HalftoneIntensity}), B=@n53)
  NamedRerouteDeclaration(None=Multiply(A=LinearInterpolate(B=@n39.RGB, Alpha=@n43) {constA=1}, B=Add(A=@n54, B=@n53))) {name=Color, nodeColor=<rgba 0 0.9062 1 1>}
  NamedRerouteDeclaration(None=@n48) {name=BaseColor_Texture, nodeColor=<rgba 0.2422 0 1 1>}
  NamedRerouteUsage()
  NamedRerouteUsage()
  n55 = Fresnel() {exponent=0.5}
  n56 = Constant() {r=0.1}
  Reroute(None=OneMinus(None=MaterialFunctionCall(In=@n55, Contrast=@n56) {materialFunction=&CheapContrast}.Result))
  TextureSampleParameter2D() {
   group=Textures
   parameterName=NormalMap
   samplerType=SAMPLERTYPE_Normal
   texture=/BRCosmetics/Accessories/FORT_Backpacks/Backpack_M_MED_ChunkTiara/Textures/T_Pimlico_Crownball_N
  }
 ]
}
